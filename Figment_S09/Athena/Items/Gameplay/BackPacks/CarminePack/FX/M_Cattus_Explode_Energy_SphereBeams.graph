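# Unlit translucent mesh-particle material for energy beams on a sphere in the CarminePack Cattus explosion effect, colored by particle color.
# Opacity multiplies Fresnel times one-minus-Fresnel bands, softening center and rim, with two panning smoke-texture samples at different tilings, particle alpha and depth fade.
# Dynamic parameters drive a Fresnel exponent (MiddleFade), pan speed and refraction strength; EdgeFade is unconnected.
# A static switch enables view-angle-faded refraction.
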Material /Figment_S09/Athena/Items/Gameplay/BackPacks/CarminePack/FX/M_Cattus_Explode_Energy_SphereBeams
graph {
 properties {
  bUsedWithMeshParticles=true
  blendMode=BLEND_Translucent
  pixelDepthOffsetMode=PDOM_Legacy
  refractionMethod=RM_IndexOfRefraction
  shadingModel=MSM_Unlit
  usedShadingModels=Unlit
 }
 nodes [
  n1 = Fresnel(ExponentIn=Constant() {r=0.8}) {baseReflectFraction=0.001}
  n2 = Multiply(A=Multiply(A=OneMinus(None=@n1), B=@n1), B=ParticleColor().A) {constA=1}
  n3 = DynamicParameter() {defaultValue=<rgba 3 1 1 1>, paramNames [MiddleFade,EdgeFade,PanSpeed,Ref]}
  n4 = Fresnel(ExponentIn=@n3.MiddleFade) {baseReflectFraction=0.001}
  n5 = Multiply(A=OneMinus(None=@n4), B=@n4)
  n6 = Constant() {r=1}
  n7 = Multiply(A=Time(), B=@n3.PanSpeed)
  n8 = TextureSample(UVs=Panner(Coordinate=TextureCoordinate(), Time=@n7) {speedY=-1}) {texture=/Game/Effects/Fort_Effects/Textures/Tile/T_Inky_Smoke_Tile}
  n9 = TextureSample(UVs=Panner(Coordinate=TextureCoordinate() {uTiling=8, vTiling=2}, Time=@n7) {speedY=-1}) {texture=&T_Inky_Smoke_Tile}
  n10 = Multiply(A=Power(Base=Multiply(A=@n5, B=@n6)), B=Multiply(A=Multiply(A=@n8.G, B=@n9.G), B=Constant() {r=4}))
  DepthFade(Opacity=Multiply(A=Power(Base=@n2), B=Clamp(None=@n10)), FadeDistance=ScalarParameter() {defaultValue=100, parameterName=DepthFadeDIst})
  ParticleColor()
  n11 = MaterialFunctionCall("FadeSensitivity (optional)"=Constant() {r=2}) {materialFunction=&FadeBasedOnViewAngle}
  StaticSwitchParameter(True=LinearInterpolate(Alpha=Multiply(A=@n3.Ref, B=@n11.Result)) {constA=1, constB=1.1}, False=Constant() {r=1}) {parameterName=UseRefraction}
 ]
}
