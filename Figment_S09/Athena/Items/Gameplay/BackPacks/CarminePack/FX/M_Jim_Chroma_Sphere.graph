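# Unlit translucent, refractive particle material for a chromatic sphere effect in the Carmine backpack FX.
# Swirl-normal-warped, panning noise is sampled at three shifted UVs to fake RGB chromatic separation, with a baked-color texture switch and dynamic-parameter offset and contrast.
# The saturation-boosted result drives HDR color and, with Fresnel and gradient masks, opacity.
# Depth and near-camera fades are skipped on mobile; a similar Fresnel mask drives refraction.
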Material /Figment_S09/Athena/Items/Gameplay/BackPacks/CarminePack/FX/M_Jim_Chroma_Sphere
graph {
 properties {
  bUsedWithMeshParticles=true
  bUsedWithParticleSprites=true
  blendMode=BLEND_Translucent
  pixelDepthOffsetMode=PDOM_Legacy
  refractionMethod=RM_IndexOfRefraction
  shadingModel=MSM_Unlit
  usedShadingModels=Unlit
 }
 nodes [
  n1 = ParticleColor()
  n2 = DynamicParameter() {paramNames [UVWarpStr,RandX,RandY,Power]}
  n3 = Multiply(A=Constant2Vector() {r=1}, B=ScalarParameter() {defaultValue=0.025, parameterName=ChromaShiftDistance})
  n4 = ScalarParameter() {defaultValue=2, parameterName=BaseTilingU}
  n5 = ScalarParameter() {defaultValue=1, parameterName=BaseTilingV}
  n6 = TextureCoordinate()
  n7 = AppendVector(A=ScalarParameter() {defaultValue=2, parameterName=WarpTexTilingU}, B=ScalarParameter() {defaultValue=1, parameterName=WarpTexTilingV})
  n8 = ScalarParameter() {parameterName=WarpPanSpeedU}
  n9 = ScalarParameter() {defaultValue=0.25, parameterName=WarpPanSpeedV}
  n10 = TextureSampleParameter2D(UVs=Panner(Coordinate=Multiply(A=@n6, B=@n7), Speed=AppendVector(A=@n8, B=@n9)) {bFractionalPart=true, speedY=0.25}) {parameterName=UVDistTex, samplerType=SAMPLERTYPE_Normal, texture=/Figment_S09/Environments/Foliage/Textures/T_SwirlNoise_N}
  n11 = DynamicParameter() {paramNames [UVWarpStr,RandX,RandY,Power]}
  n12 = ScalarParameter() {defaultValue=0.07, parameterName=UVWarpStr}
  n13 = Add(
   A=Multiply(A=TextureCoordinate(), B=AppendVector(A=@n4, B=@n5))
   B=Multiply(
    A=ComponentMask(None=@n10.RGB) {g=true, r=true}
    B=StaticSwitchParameter(True=@n11.UVWarpStr, False=@n12) {defaultValue=true, parameterName=UseDynaParam_DistortUv}
   ) {constB=0.0125}
  )
  n14 = Panner(
   Coordinate=Add(A=AppendVector(A=@n2.RandX, B=@n2.RandX), B=Add(A=@n13, B=@n3))
   Speed=AppendVector(A=ScalarParameter() {defaultValue=0.25, parameterName=PanSpeedU}, B=ScalarParameter() {defaultValue=0.5, parameterName=PanSpeedV})
  ) {bFractionalPart=true, speedY=0.25}
  n15 = TextureSampleParameter2D(UVs=@n14) {parameterName=GrayscaleTex, texture=/Figment_S09/Effects/Fort_Effects/Textures/Tile/T_FireNoise_01}
  n16 = DynamicParameter() {paramNames [UVWarpStr,RandX,RandY,Power]}
  n17 = DynamicParameter() {paramNames [UVWarpStr,RandX,RandY,Power]}
  n18 = AppendVector(A=@n17.RandX, B=@n17.RandX)
  n19 = ScalarParameter() {defaultValue=0.25, parameterName=PanSpeedU}
  n20 = ScalarParameter() {defaultValue=0.5, parameterName=PanSpeedV}
  n21 = TextureSampleParameter2D(UVs=Panner(Coordinate=Add(A=@n18, B=@n13), Speed=AppendVector(A=@n19, B=@n20)) {bFractionalPart=true, speedY=0.25}) {parameterName=GrayscaleTex_1, texture=&T_FireNoise_01}
  n22 = DynamicParameter() {paramNames [UVWarpStr,RandX,RandY,Power]}
  n23 = DynamicParameter() {paramNames [UVWarpStr,RandX,RandY,Power]}
  n24 = AppendVector(A=@n23.RandX, B=@n23.RandX)
  n25 = Add(A=@n13, B=Multiply(A=ScalarParameter() {defaultValue=0.025, parameterName=ChromaShiftDistance}, B=Constant2Vector() {r=-1}))
  n26 = ScalarParameter() {defaultValue=0.25, parameterName=PanSpeedU}
  n27 = ScalarParameter() {defaultValue=0.5, parameterName=PanSpeedV}
  n28 = TextureSampleParameter2D(UVs=Panner(Coordinate=Add(A=@n24, B=@n25), Speed=AppendVector(A=@n26, B=@n27)) {bFractionalPart=true, speedY=0.25}) {parameterName=GrayscaleTex_2, texture=&T_FireNoise_01}
  n29 = DynamicParameter() {paramNames [UVWarpStr,RandX,RandY,Power]}
  n30 = StaticSwitchParameter(
   True=Power(
    Base=TextureSampleParameter2D(UVs=@n14) {parameterName=ColorTex, texture=/Game/Effects/Fort_Effects/Textures/Tile/T_Inky_Smoke_Tile_Chromatic}.RGB
    Exp=DynamicParameter() {paramNames [UVWarpStr,RandX,RandY,Power]}.Power
   ) {constExponent=14}
   False=MaterialFunctionCall(
    X=Power(Base=@n15.G, Exp=@n16.Power) {constExponent=14}
    Y=Power(Base=@n21.G, Exp=@n22.Power) {constExponent=14}
    Z=Power(Base=@n28.G, Exp=@n29.Power) {constExponent=14}
   ) {materialFunction=&MakeFloat3}.Result
  ) {parameterName=TexHasChromaBakedIn}
  n31 = OneMinus(None=Fresnel() {baseReflectFraction=0, exponent=1})
  n32 = Multiply(A=@n31, B=@n31)
  n33 = DynamicParameter() {paramNames [UVWarpStr,RandX,RandY,Power]}
  n34 = ScalarParameter() {parameterName=CenterOpacity}
  n35 = Multiply(
   A=Multiply(A=@n32, B=@n32)
   B=Fresnel(BaseReflectFractionIn=StaticSwitchParameter(True=@n33.UVWarpStr, False=@n34) {desc=Warning: Turn off Dyn UVWrapStr, parameterName=Allow DynCenterOpacity}) {baseReflectFraction=0, exponent=1}
  )
  n36 = TextureSample(UVs=MaterialFunctionCall(XY=TextureCoordinate()) {materialFunction=&Swizzle}.YX) {samplerType=SAMPLERTYPE_LinearColor, texture=/Game/Weapons/WeaponSkins/Textures/T_VertGradient}
  n37 = Multiply(A=@n1.A, B=Multiply(A=@n30, B=Multiply(A=@n35, B=@n36.RGB)))
  n38 = Multiply(
   A=DepthFade(Opacity=@n37, FadeDistance=ScalarParameter() {defaultValue=600, parameterName=DepthFadeDistance})
   B=MaterialFunctionCall(Distance=ScalarParameter() {defaultValue=25, parameterName=NearCamFade}) {materialFunction=&Near_Camera_Fade}.Result
  )
  FeatureLevelSwitch(Default=@n38, ES2=@n37, ES3_1=@n37, SM4_REMOVED=@n38)
  n39 = OneMinus(None=Fresnel() {baseReflectFraction=0, exponent=1})
  n40 = Multiply(A=@n39, B=@n39)
  n41 = Multiply(A=@n40, B=@n40)
  n42 = Fresnel(BaseReflectFractionIn=ScalarParameter() {parameterName=CenterOpacity_1}) {baseReflectFraction=0, exponent=1}
  LinearInterpolate(B=ScalarParameter() {defaultValue=1.1, parameterName=RefractionStr}, Alpha=Multiply(A=Multiply(A=@n41, B=@n42), B=ParticleColor().A)) {constA=1, constB=1.1}
  n43 = Desaturation(None=@n30, Fraction=ScalarParameter() {defaultValue=-3, parameterName=SaturationFactor})
  MaterialFunctionCall(LDR Input=Clamp(None=@n43), HDR Tint=Multiply(A=@n43, B=@n1.RGB)) {materialFunction=&DeriveHDRfromLDR}
 ]
}
