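# Translucent, unlit, two-sided mesh-particle material for a soft smoky cone, apparently a punch effect for the CarminePack backpack.
# Four panning chromatic smoke samples are blended as noise, driven by Time or a dynamic-parameter TimeOffset via static switch.
# A V-gradient blends tinted into plain particle color and drives hue shift, multiplied by oversaturated smoke.
# Opacity combines smoke, particle alpha and a sphere mask positioned by dynamic PannerX, with depth and near-camera fades.
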Material /Figment_S09/Athena/Items/Gameplay/BackPacks/CarminePack/FX/M_Jim_Punch_Cone_Soft
graph {
 properties {
  bUsedWithMeshParticles=true
  blendMode=BLEND_Translucent
  pixelDepthOffsetMode=PDOM_Legacy
  shadingModel=MSM_Unlit
  twoSided=true
  usedShadingModels=Unlit
 }
 nodes [
  n1 = ComponentMask(None=TextureCoordinate()) {g=true}
  n2 = Clamp(None=Power(Base=Multiply(A=@n1, B=@n1)))
  n3 = Multiply(A=ScalarParameter() {defaultValue=0.5, parameterName=HueShiftEdge}, B=@n2) {constA=0.7}
  n4 = ParticleColor()
  n5 = Multiply(A=VectorParameter() {parameterName=ColorTint, defaultValue=<rgba 1 0.5 0 1>}.RGB, B=@n4.RGB)
  n6 = DynamicParameter() {defaultValue=<rgba 0 0.5 1 1>, paramNames [PannerX,PannerY,Param3,TimeOffset]}
  n7 = StaticSwitchParameter(True=@n6.TimeOffset, False=Time()) {parameterName=DP4ForTime}
  n8 = Panner(Coordinate=TextureCoordinate() {uTiling=2, vTiling=0.25}, Time=@n7) {speedX=-0.5, speedY=-1}
  n9 = Panner(Coordinate=TextureCoordinate() {uTiling=2, vTiling=0.5}, Time=@n7) {speedX=1, speedY=-2}
  n10 = Panner(Coordinate=TextureCoordinate() {uTiling=2, vTiling=0.5}, Time=@n7) {speedX=-0.5, speedY=-1}
  n11 = Panner(Coordinate=TextureCoordinate() {uTiling=3}, Time=@n7) {speedX=1, speedY=-2}
  n12 = LinearInterpolate(
   A=LinearInterpolate(
    A=TextureSample(UVs=@n8) {texture=/Game/Effects/Fort_Effects/Textures/Tile/T_Inky_Smoke_Tile_Chromatic}.RGB
    B=TextureSample(UVs=@n9) {texture=&T_Inky_Smoke_Tile_Chromatic}.RGB
   )
   B=LinearInterpolate(
    A=TextureSample(UVs=@n10) {texture=&T_Inky_Smoke_Tile_Chromatic}.RGB
    B=TextureSample(UVs=@n11) {texture=&T_Inky_Smoke_Tile_Chromatic}.RGB
   )
  )
  Multiply(
   A=MaterialFunctionCall(Hue Shift Percentage=Clamp(None=@n3), Texture=LinearInterpolate(A=@n5, B=@n4.RGB, Alpha=@n2)) {materialFunction=&HueShift}.Result
   B=Desaturation(None=@n12, Fraction=Constant() {r=-0.5})
  ) {constB=1.5}
  n13 = Multiply(
   A=@n4.A
   B=SphereMask(
    A=TextureCoordinate() {uTiling=0}
    B=AppendVector(A=@n6.PannerX, B=@n6.PannerX)
    Radius=ScalarParameter() {defaultValue=0.25, parameterName=SoftMask_Radius}
    Hardness=ScalarParameter() {defaultValue=-1, parameterName=SoftMask_Hardness}
   )
  )
  Multiply(
   A=MaterialFunctionCall(Distance=ScalarParameter() {defaultValue=250, parameterName=Camera_FadeDistance}) {materialFunction=&Near_Camera_Fade}.Result
   B=DepthFade(Opacity=Multiply(A=@n13, B=@n12), FadeDistance=ScalarParameter() {defaultValue=250, parameterName=Depth_FadeDistance})
  )
 ]
}
